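; BP_TimeShimmer is a visual effect Blueprint for the Rift foraged item, showing a shimmer with a cross flare on the SM_Flare component.
; The construction script creates two dynamic material instances on that component: a geode shimmer in slot 0 and a sharp flare in slot 1.
; It feeds the Opacity and FlareSize variables into their matching scalar parameters.
; BeginPlay, overlap and Tick events are empty.
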(event EventBeginPlay)

(event Collision|EventActorBeginOverlap
  (params (OtherActor :type Actor%20Object%20Reference)))

(event EventTick
  (params (DeltaSeconds :type Float%20%28single-precision%29)))

(fn ConstructionScript
  (bind _sm_flare (Variables|Default|GetSM_Flare))
  (bind _returnvalue (Rendering|Material|CreateDynamicMaterialInstance _sm_flare 1 "/Figment_S10/Effects/Fort_Effects/Materials/Flares/M_Sharp_Cross_Flare_Inst.M_Sharp_Cross_Flare_Inst" "FlareSize" :!class /Script/Engine.PrimitiveComponent))
  (Variables|Default|SetMID _returnvalue)
  (Rendering|Material|SetScalarParameterValue (Variables|Default|GetMID) "FlareSize" (Variables|Default|GetFlareSize) :!class /Script/Engine.MaterialInstanceDynamic)
  (bind _returnvalue_2 (Rendering|Material|CreateDynamicMaterialInstance _sm_flare 0 "/Figment_S10/Athena/Effects/Materials/Geode/M_Geode_Oddity_Shimmer_02.M_Geode_Oddity_Shimmer_02" "Opacity" :!class /Script/Engine.PrimitiveComponent))
  (Variables|Default|SetMIC _returnvalue_2)
  (Rendering|Material|SetScalarParameterValue (Variables|Default|GetMIC) "Opacity" (Variables|Default|GetOpacity) :!class /Script/Engine.MaterialInstanceDynamic))
